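# Physics asset for the Balloon pickaxe skeletal mesh, giving its root and two dyn_ bones simple collision bodies for physics-driven motion.
# A long, thin capsule covers b_AR_Root, a fatter one covers dyn_head_01, and a zero-length capsule acts as a sphere on dyn_tail_01.
# Head and tail each constrain directly to the root rather than chaining, with small angular limit values that are much tighter on the tail.
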PhysicsAsset /BRCosmetics/Weapons/FORT_Melee/Pickaxe_Balloon/Meshes/SK_Pickaxe_Balloon_PhysicsAsset
bodies {
 b_AR_Root [{shapeType=Capsule, length=66.3876, radius=2.7073, center=<7.5909 -0.7703 -16.6324>, rotation=<rot -0.0487 -90 0.2617>}]
 dyn_head_01 [{shapeType=Capsule, length=48.5, radius=7.5, center=<0 0 21.8207>, rotation=<rot 0 0 -90.0001>}]
 dyn_tail_01 [{shapeType=Capsule, length=0, radius=5, center=<0 0 -9.682>}]
}
constraints [
 {bone1Name=dyn_head_01, bone2Name=b_AR_Root, swing1LimitDegrees=15, swing2LimitDegrees=0, twistLimitDegrees=15}
 {bone1Name=dyn_tail_01, bone2Name=b_AR_Root, swing1LimitDegrees=1, swing2LimitDegrees=0, twistLimitDegrees=1}
]
